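# Niagara system for the Electrobeam vehicle boost cosmetic, emitting a beam trail from filtered skeletal-mesh sockets.
# Three ribbon emitters layer streaks that shoot backward and wobble through time-driven sine offsets on position, with curve-shaped width.
# A sprite emitter adds two spawn groups of curl-noise sparks, while a second sprite emitter is disabled.
# Default colors are mostly glowing green-teal, with per-paint-variant overrides when the boost is painted.
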NiagaraSystem /VehicleCosmetics/Boosts/Boost_Electrobeam/FX/NS_Boost_Electrobeam
dependencies {
 usedDataInterfaces [
  &NiagaraDataInterfaceCurve
  &NiagaraDataInterfaceArrayFloat
  &NiagaraDataInterfaceVectorCurve
  &NiagaraDataInterfaceSkeletalMesh
 ]
 usedDynamicInputs []
 usedModules [
  &self:SystemSpawnScript.SystemScriptSource.SystemScriptGraph.NiagaraNodeAssignment_0.SetVariables_#1
  &SystemState
 ]
}
emitters {
 NE_VC_Boost_Ribbon_Parent {
  emitterSpawnScript [/SetVariables(
    Emitter.IsPainted:NiagaraFloat=&MakeCustomFloatFromBool(Bool=$User.bIsPainted, True Float=2, False Float=-0.512)
    Emitter.MirrorOffset:NiagaraFloat=&MakeCustomFloatFromBool(Bool=$User.bIsPainted, True Float=4, False Float=2)
   ) {isSetParameters=true, module=SetVariables_#2}]
  emitterUpdateScript [
   &EmitterState(Life Cycle Mode=::NewEnumerator1, Loop Duration=1, UseLoopDelay=false)
   &VC_FilteredSocketCount()
   &VC_RateQualityHandler()
   /SetVariables(
    Emitter.AnotherOffset:NiagaraFloat=&Power(a=HlslExpression() {hlslExpression="(fmod(Engine.Time * 0.35, 0.2) + 0.0) * 3"}, b=1)
    Emitter.TimeLerp:NiagaraFloat=&Power(a=HlslExpression() {hlslExpression="(fmod(Engine.Time * 0.325, 0.4) + 0.15) * 3"}, b=2)
    Emitter.TimeOffset:NiagaraFloat=&Power(a=HlslExpression() {hlslExpression="fmod(Engine.Time * 0.5, 0.3) + 0.15 * 3"}, b=2)
   ) {isSetParameters=true, module=SetVariables_#3}
   /SetVariables(
    Emitter.LerpChannel:Vector3f=&Lerp_Vector(
     A=<1 0 0>
     B=<0 1 0>
     Alpha=&Ceil_Float(Float=&Sine(Normalized Angle=$Emitter.Age, Period=$Emitter.TimeLerp, Scale=1, Bias=0))
    )
    Emitter.OffsetX:NiagaraFloat=&Multiply_Float(
     A=&Ceil_Float(Float=&Sine(Normalized Angle=$Emitter.Age, Period=$Emitter.TimeOffset, Scale=1, Bias=0))
     B=0.25
    )
   ) {isSetParameters=true, module=SetVariables_#4}
   &VC_SpawnRibbon_v1(Number of Ribbons=$Emitter.PortCount, SpawnRate=30)
  ]
  particleSpawnScript [
   /VehicleCosmetics/FX/Niagara/Modules/MergeEmitters/VC_ExtraBindingColors(
    ExtraColors=2
    R1 Color:LinearColor=&MakeCustomLinearColorFromBool(Bool=$User.bIsPainted, True Color=$User.PaintedColor, False Color=<rgba 0.4988 24 12.5029 1>)
    R2 Color:LinearColor=&MakeCustomLinearColorFromBool(Bool=$User.bIsPainted, True Color=$User.PaintedColor, False Color=<rgba 0.1247 6 3.1257 1>)
   )
   &VC_PaintedVariantsOverride(
    R1 Black:LinearColor=<rgba 3 3 3 1>
    R1 BurntSienna:LinearColor=<rgba 20 8 5 1>
    R1 Cobalt:LinearColor=<rgba 6.4 7.2533 32 1>
    R1 Crimson:LinearColor=<rgba 33 6.6 7.3156 1>
    R1 DefaultColor=<rgba 0.4988 24 12.5029 1>
    R1 ForestGreen:LinearColor=<rgba 6.875 22 5.5 1>
    R1 Grey:LinearColor=<rgba 10.4132 10.4132 12 1>
    R1 Lime:LinearColor=<rgba 18.0267 26 5.2 1>
    R1 Orange:LinearColor=<rgba 35 11.3556 7 1>
    R1 Pink:LinearColor=<rgba 28 8.96 28 1>
    R1 Purple:LinearColor=<rgba 10.032 0.88 22 1>
    R1 Saffron:LinearColor=<rgba 22 18.5093 7.04 1>
    R1 SkyBlue:LinearColor=<rgba 4.4 16.1108 22 1>
    R1 TitaniumWhite:LinearColor=<rgba 16 16 16 1>
    R2 Black:LinearColor=<rgba 0.75 0.75 0.75 1>
    R2 BurntSienna:LinearColor=<rgba 4 1.44 0.8 1>
    R2 Cobalt:LinearColor=<rgba 0.96 1.128 6 1>
    R2 Crimson:LinearColor=<rgba 6 1.32 1.4468 1>
    R2 DefaultColor:LinearColor=<rgba 0.1247 6 3.1257 1>
    R2 ForestGreen:LinearColor=<rgba 1.875 6 1.5 1>
    R2 Grey:LinearColor=<rgba 1.7355 1.7355 2 1>
    R2 Lime:LinearColor=<rgba 4.16 6 1.2 1>
    R2 Orange:LinearColor=<rgba 10 3.2444 2 1>
    R2 Pink:LinearColor=<rgba 10 3.2 10 1>
    R2 Purple:LinearColor=<rgba 2.668 0.12 6 1>
    R2 Saffron:LinearColor=<rgba 6 4.908 1.32 1>
    R2 SkyBlue:LinearColor=<rgba 1.44 5.8049 8 1>
    R2 TitaniumWhite:LinearColor=<rgba 6 6 6 1>
    Use R1 Painted Overrides=true
    Use R2 Painted Overrides=true
   )
   /SetVariables(
    Particles.ColorTwo:LinearColor=<rgba 0.1247 6 3.1257 1>
    Particles.SizeTwo:NiagaraFloat=&RandomRangeFloat(Minimum=60, Maximum=60)
   ) {isSetParameters=true, module=SetVariables_#5}
   &VC_InitializeRibbon_v1(
    Color=&MakeCustomLinearColorFromBool(Bool=$User.bIsPainted, True Color=$User.PaintedColor, False Color=<rgba 0.4988 24 12.5029 1>)
    Custom Offset=true
    Enable Skeletal Position=true
    Lifetime=0.46
    Offset Amount=1
    Ribbon Width:NiagaraFloat=&RandomRangeFloat(Minimum=90, Maximum=90)
   )
   &AddVelocity(Velocity=<-2500 0 50>) {module=AddVelocity001}
   &ShapeLocation(MaskSpawnGroup=false, Override Seed=false, Sphere Radius=7, UseOwnerScale=true) {enabled=false}
  ]
  particleUpdateScript [
   &ParticleState()
   &ScaleColor(
    Scale Alpha=&MakeCustomFloatFromBool(
     Bool=$User.bIsPainted
     True Float=0.15
     False Float=&FloatFromCurve(FloatCurve=DataInterface() {propertyValues {Curve {keys [<at 0 0>,<at 0.02 0.1>,<at 0.95 0.1>,<at 1 0>]}}}, Scale Curve=0.2)
    )
    ScaleA=true
    ScaleRGB=true
   ) {enabled=false}
   /SetVariables(Particles.R1 Color:LinearColor=/DealershipSharedLibrary/FX/Niagara/DynamicInputs/MultiplyLinearColorByVectorAndFloat(
     Color to Scale:LinearColor=$Particles.Initial.R1 Color
     RGB Scale:Vector3f=<1 1 1>
     Alpha Scale:NiagaraFloat=&MakeCustomFloatFromBool(
      Bool=$User.bIsPainted
      True Float=&FloatFromCurve(FloatCurve=DataInterface() {propertyValues {Curve {keys [<at 0 0>,<at 0.02 0.1>,<at 0.95 0.1>,<at 1 0>]}}}, Scale Curve=0.2)
      False Float=&FloatFromCurve(FloatCurve=DataInterface() {propertyValues {Curve {keys [<at 0 0>,<at 0.02 0.1>,<at 0.95 0.1>,<at 1 0>]}}}, Scale Curve=0.2)
     )
    )) {isSetParameters=true, module=SetVariables_#6}
   &ScaleRibbonWidth(Ribbon Width Scale=&FloatFromCurve(
     FloatCurve=DataInterface() {propertyValues {Curve {keys [
         {interpMode=RCIM_Cubic, tangentMode=RCTM_User, time=0, value=0}
         {arriveTangent=0.5028, interpMode=RCIM_Cubic, leaveTangent=0.5028, tangentMode=RCTM_User, time=0.1, value=1.25}
         {arriveTangent=-0.5446, interpMode=RCIM_Cubic, leaveTangent=-0.5446, tangentMode=RCTM_User, time=0.8, value=1}
         {arriveTangent=-3.985, interpMode=RCIM_Cubic, leaveTangent=-3.985, tangentMode=RCTM_User, time=0.9, value=0}
        ]}}}
     Scale Curve=1.25
    ))
   /SetVariables(Particles.SizeScaleTwo:NiagaraFloat=&Multiply_Float(
     A=$Particles.SizeTwo
     B=&FloatFromCurve(FloatCurve=DataInterface() {propertyValues {Curve {keys [{arriveTangent=-0.0009, time=0, value=1}, {interpMode=RCIM_Cubic, leaveTangent=-2.5286, time=1, value=1}]}}})
    )) {isSetParameters=true, module=SetVariables_#7}
   &SolveForcesAndVelocity(Clamp Velocity=false, Limit Acceleration=false)
   &VC_UpdateRibbon_v1(Enable Skeletal Position=true)
   /SetVariables(Particles.CustomPosition:NiagaraPosition=&AddVectorToPosition(
     Position=$Particles.Position
     Vector=&TransformVector(
      Vector=&MakeVector(
       X=0
       Y=&Sine(
        Normalized Angle=&Multiply_Float(
         A=1
         B=&Subtract_Float(
          A=&Multiply_Float(A=2, B=&Subtract_Float(A=$Engine.Time, B=0.5))
          B=$Particles.NormalizedAge
         )
        )
        Period=&SelectFloatFromArray(
         Direct Array Index=&SpawnGroup()
         Float Selection Array=DataInterface() {propertyValues {FloatData [0.7,0.765]}}
        )
        Scale=&FloatFromCurve(FloatCurve=DataInterface() {propertyValues {Curve {keys [<at 0 60 RCIM_Cubic>,<at 0.86 0 RCIM_Cubic>]}}})
        Bias=0
       )
       Z=&Sine(
        Normalized Angle=&Multiply_Float(
         A=1
         B=&Subtract_Float(
          A=&Multiply_Float(A=2, B=&Subtract_Float(A=$Engine.Time, B=0.5))
          B=$Particles.NormalizedAge
         )
        )
        Period=&SelectFloatFromArray(
         Direct Array Index=&SpawnGroup()
         Float Selection Array=DataInterface() {propertyValues {FloatData [0.3,0.365]}}
        )
        Scale=&FloatFromCurve(FloatCurve=DataInterface() {propertyValues {Curve {keys [<at 0.1 2>,<at 1 48>]}}})
        Bias=0
       )
      )
      Destination Space=::Simulation
      Apply Scale=true
     )
    )) {isSetParameters=true}
  ]
  rendererClasses [&NiagaraRibbonRendererProperties]
  renderers [
   {rendererClass=&NiagaraRibbonRendererProperties}
   {rendererClass=&NiagaraRibbonRendererProperties, rendererIndex=1}
   {rendererClass=&NiagaraRibbonRendererProperties, rendererIndex=2}
  ]
 }
 NE_VC_Boost_Ribbon_Parent007 {
  emitterSpawnScript [/SetVariables(
    Emitter.IsPainted:NiagaraFloat=&MakeCustomFloatFromBool(Bool=$User.bIsPainted, True Float=2, False Float=-0.512)
    Emitter.MirrorOffset:NiagaraFloat=&MakeCustomFloatFromBool(Bool=$User.bIsPainted, True Float=4, False Float=2)
   ) {isSetParameters=true, module=SetVariables_#2}]
  emitterUpdateScript [
   &EmitterState(Life Cycle Mode=::NewEnumerator1, Loop Duration=1, UseLoopDelay=false)
   &VC_FilteredSocketCount()
   &VC_RateQualityHandler()
   /SetVariables(
    Emitter.AnotherOffset:NiagaraFloat=&Power(a=HlslExpression() {hlslExpression="(fmod(Engine.Time * 0.35, 0.2) + 0.0) * 3"}, b=1)
    Emitter.TimeLerp:NiagaraFloat=&Power(a=HlslExpression() {hlslExpression="(fmod(Engine.Time * 0.325, 0.4) + 0.15) * 3"}, b=2)
    Emitter.TimeOffset:NiagaraFloat=&Power(a=HlslExpression() {hlslExpression="fmod(Engine.Time * 0.5, 0.3) + 0.15 * 3"}, b=2)
   ) {isSetParameters=true, module=SetVariables_#3}
   /SetVariables(
    Emitter.LerpChannel:Vector3f=&Lerp_Vector(
     A=<1 0 0>
     B=<0 1 0>
     Alpha=&Ceil_Float(Float=&Sine(Normalized Angle=$Emitter.Age, Period=$Emitter.TimeLerp, Scale=1, Bias=0))
    )
    Emitter.OffsetX:NiagaraFloat=&Multiply_Float(
     A=&Ceil_Float(Float=&Sine(Normalized Angle=$Emitter.Age, Period=$Emitter.TimeOffset, Scale=1, Bias=0))
     B=0.25
    )
   ) {isSetParameters=true, module=SetVariables_#4}
   &VC_SpawnRibbon_v1(Number of Ribbons=$Emitter.PortCount, SpawnRate=120)
  ]
  particleSpawnScript [
   &VC_ExtraBindingColors(
    ExtraColors=2
    R1 Color:LinearColor=&MakeCustomLinearColorFromBool(Bool=$User.bIsPainted, True Color=$User.PaintedColor, False Color=<rgba 0.4988 24 12.5029 1>)
    R2 Color:LinearColor=&MakeCustomLinearColorFromBool(Bool=$User.bIsPainted, True Color=$User.PaintedColor, False Color=<rgba 0.1247 6 3.1257 1>)
   )
   &VC_PaintedVariantsOverride(
    R1 Black:LinearColor=<rgba 3 3 3 1>
    R1 BurntSienna:LinearColor=<rgba 20 8 5 1>
    R1 Cobalt:LinearColor=<rgba 6.4 7.2533 32 1>
    R1 Crimson:LinearColor=<rgba 33 6.6 7.3156 1>
    R1 DefaultColor=<rgba 0.4988 24 12.5029 1>
    R1 ForestGreen:LinearColor=<rgba 6.875 22 5.5 1>
    R1 Grey:LinearColor=<rgba 10.4132 10.4132 12 1>
    R1 Lime:LinearColor=<rgba 18.0267 26 5.2 1>
    R1 Orange:LinearColor=<rgba 35 11.3556 7 1>
    R1 Pink:LinearColor=<rgba 28 8.96 28 1>
    R1 Purple:LinearColor=<rgba 10.032 0.88 22 1>
    R1 Saffron:LinearColor=<rgba 22 18.5093 7.04 1>
    R1 SkyBlue:LinearColor=<rgba 4.4 16.1108 22 1>
    R1 TitaniumWhite:LinearColor=<rgba 16 16 16 1>
    R2 Black:LinearColor=<rgba 0.75 0.75 0.75 1>
    R2 BurntSienna:LinearColor=<rgba 4 1.44 0.8 1>
    R2 Cobalt:LinearColor=<rgba 0.96 1.128 6 1>
    R2 Crimson:LinearColor=<rgba 6 1.32 1.4468 1>
    R2 DefaultColor:LinearColor=<rgba 0.1247 6 3.1257 1>
    R2 ForestGreen:LinearColor=<rgba 1.875 6 1.5 1>
    R2 Grey:LinearColor=<rgba 1.7355 1.7355 2 1>
    R2 Lime:LinearColor=<rgba 4.16 6 1.2 1>
    R2 Orange:LinearColor=<rgba 10 3.2444 2 1>
    R2 Pink:LinearColor=<rgba 10 3.2 10 1>
    R2 Purple:LinearColor=<rgba 2.668 0.12 6 1>
    R2 Saffron:LinearColor=<rgba 6 4.908 1.32 1>
    R2 SkyBlue:LinearColor=<rgba 1.44 5.8049 8 1>
    R2 TitaniumWhite:LinearColor=<rgba 6 6 6 1>
    Use R1 Painted Overrides=true
    Use R2 Painted Overrides=true
   )
   /SetVariables(
    Particles.ColorTwo:LinearColor=<rgba 0.1247 6 3.1257 1>
    Particles.SizeTwo:NiagaraFloat=&RandomRangeFloat(Minimum=20, Maximum=30)
   ) {isSetParameters=true, module=SetVariables_#5}
   &VC_InitializeRibbon_v1(
    Color=&MakeCustomLinearColorFromBool(Bool=$User.bIsPainted, True Color=$User.PaintedColor, False Color=<rgba 0.4988 24 12.5029 1>)
    Custom Offset=true
    Enable Skeletal Position=true
    Lifetime=0.46
    Offset Amount=1
    Ribbon Width:NiagaraFloat=&RandomRangeFloat(Minimum=90, Maximum=90)
   )
   &AddVelocity(Velocity=<-2500 0 50>) {module=AddVelocity001}
   &ShapeLocation(MaskSpawnGroup=false, Override Seed=false, Sphere Radius=7, UseOwnerScale=true) {enabled=false}
   /SetVariables(Particles.MainRibbonWidth:NiagaraFloat=&MaskFloat(Mask Source=0, Mask Parameter=?, Mask=0, Masked Float=80, Passthrough Float=-80)) {isSetParameters=true}
  ]
  particleUpdateScript [
   &ParticleState()
   &ScaleColor(
    Scale Alpha=&MakeCustomFloatFromBool(
     Bool=$User.bIsPainted
     True Float=0.15
     False Float=&FloatFromCurve(FloatCurve=DataInterface() {propertyValues {Curve {keys [<at 0 0>,<at 0.02 0.1>,<at 0.95 0.1>,<at 1 0>]}}}, Scale Curve=0.2)
    )
    ScaleA=true
    ScaleRGB=true
   ) {enabled=false}
   /SetVariables(Particles.R1 Color:LinearColor=&MultiplyLinearColorByVectorAndFloat(
     Color to Scale:LinearColor=$Particles.Initial.R1 Color
     RGB Scale:Vector3f=<1 1 1>
     Alpha Scale:NiagaraFloat=&MakeCustomFloatFromBool(
      Bool=$User.bIsPainted
      True Float=&FloatFromCurve(FloatCurve=DataInterface() {propertyValues {Curve {keys [<at 0 0>,<at 0.02 0.1>,<at 0.95 0.1>,<at 1 0>]}}}, Scale Curve=0.2)
      False Float=&FloatFromCurve(FloatCurve=DataInterface() {propertyValues {Curve {keys [<at 0 0>,<at 0.02 0.1>,<at 0.95 0.1>,<at 1 0>]}}}, Scale Curve=0.2)
     )
    )) {isSetParameters=true, module=SetVariables_#6}
   &ScaleRibbonWidth(Ribbon Width Scale=&FloatFromCurve(
     FloatCurve=DataInterface() {propertyValues {Curve {keys [
         {interpMode=RCIM_Cubic, tangentMode=RCTM_User, time=0, value=0}
         {arriveTangent=0.5028, interpMode=RCIM_Cubic, leaveTangent=0.5028, tangentMode=RCTM_User, time=0.1, value=1.25}
         {arriveTangent=-0.5446, interpMode=RCIM_Cubic, leaveTangent=-0.5446, tangentMode=RCTM_User, time=0.8, value=1}
         {arriveTangent=-3.985, interpMode=RCIM_Cubic, leaveTangent=-3.985, tangentMode=RCTM_User, time=0.9, value=0}
        ]}}}
     Scale Curve=1.25
    ))
   /SetVariables(
    Particles.CustomWidth:NiagaraFloat=&Multiply_Float(
     A=$Particles.MainRibbonWidth
     B=&FloatFromCurve(
      FloatCurve=DataInterface() {propertyValues {Curve {keys [
          {interpMode=RCIM_Cubic, tangentMode=RCTM_User, time=0, value=0}
          {arriveTangent=0.5028, interpMode=RCIM_Cubic, leaveTangent=0.5028, tangentMode=RCTM_User, time=0.1, value=1.25}
          {arriveTangent=-0.5446, interpMode=RCIM_Cubic, leaveTangent=-0.5446, tangentMode=RCTM_User, time=0.8, value=1}
          {arriveTangent=-3.985, interpMode=RCIM_Cubic, leaveTangent=-3.985, tangentMode=RCTM_User, time=0.9, value=0}
         ]}}}
      Scale Curve=1.25
     )
    )
    Particles.SizeScaleTwo:NiagaraFloat=&Multiply_Float(
     A=$Particles.SizeTwo
     B=&FloatFromCurve(FloatCurve=DataInterface() {propertyValues {Curve {keys [
          {arriveTangent=-0.0009, time=0, value=0}
          <at 0.075 0.75 RCIM_Cubic>
          {arriveTangent=-1, leaveTangent=-1, time=0.175, value=0}
          <at 0.65 0>
          <at 0.825 0 RCIM_Cubic>
          {interpMode=RCIM_Cubic, leaveTangent=-2.5286, time=1, value=0}
         ]}}})
    )
   ) {isSetParameters=true, module=SetVariables_#7}
   &SolveForcesAndVelocity(Clamp Velocity=false, Limit Acceleration=false)
   &VC_UpdateRibbon_v1(Enable Skeletal Position=true)
   /SetVariables(Particles.Position=&AddVectorToPosition(
     Position=$Particles.Position
     Vector=&TransformVector(
      Vector=&MakeVector(
       X=0
       Y=&Sine(
        Normalized Angle=&Multiply_Float(
         A=&SelectFloatFromArray(Direct Array Index=&SpawnGroup(), Float Selection Array=DataInterface() {propertyValues {FloatData [-1,1]}})
         B=&Subtract_Float(
          A=&Multiply_Float(A=2, B=&Subtract_Float(A=$Engine.Time, B=0.5))
          B=$Particles.NormalizedAge
         )
        )
        Period=1
        Scale=&FloatFromCurve(FloatCurve=DataInterface() {propertyValues {Curve {keys [<at 0.5 8 RCIM_Cubic>,<at 1 6 RCIM_Cubic>]}}})
        Bias=0
       )
       Z=&Sine(
        Normalized Angle=&Multiply_Float(
         A=1
         B=&Subtract_Float(
          A=&Multiply_Float(A=2, B=&Subtract_Float(A=$Engine.Time, B=0.5))
          B=$Particles.NormalizedAge
         )
        )
        Period=1
        Scale=&FloatFromCurve(FloatCurve=DataInterface() {propertyValues {Curve {keys [<at 0.5 4 RCIM_Cubic>,<at 1 0 RCIM_Cubic>]}}})
        Bias=0
       )
      )
      Destination Space=::Simulation
      Apply Scale=true
     )
    )) {enabled=false, isSetParameters=true}
   /SetVariables(Particles.Position=&AddVectorToPosition(
     Position=$Particles.Position
     Vector=&TransformVector(
      Vector=&MakeVector(
       X=0
       Y=&Sine(
        Normalized Angle=&Multiply_Float(
         A=1
         B=&Subtract_Float(
          A=&Multiply_Float(A=2, B=&Subtract_Float(A=$Engine.Time, B=0.5))
          B=$Particles.NormalizedAge
         )
        )
        Period=&SelectFloatFromArray(
         Direct Array Index=&SpawnGroup()
         Float Selection Array=DataInterface() {propertyValues {FloatData [0.7,0.765]}}
        )
        Scale=&FloatFromCurve(FloatCurve=DataInterface() {propertyValues {Curve {keys [<at 0 10 RCIM_Cubic>,<at 0.86 0 RCIM_Cubic>]}}})
        Bias=0
       )
       Z=&Sine(
        Normalized Angle=&Multiply_Float(
         A=1
         B=&Subtract_Float(
          A=&Multiply_Float(A=2, B=&Subtract_Float(A=$Engine.Time, B=0.5))
          B=$Particles.NormalizedAge
         )
        )
        Period=1
        Scale=&FloatFromCurve(FloatCurve=DataInterface() {propertyValues {Curve {keys [<at 0.25 4 RCIM_Cubic>,<at 0.5 0 RCIM_Cubic>]}}})
        Bias=0
       )
      )
      Destination Space=::Simulation
      Apply Scale=true
     )
    )) {isSetParameters=true}
   /SetVariables(Particles.CustomPosition:NiagaraPosition=&AddVectorToPosition(
     Position=$Particles.Position
     Vector=&TransformVector(
      Vector=&MakeVector(
       X=0
       Y=&Sine(
        Normalized Angle=&Multiply_Float(
         A=&SelectFloatFromArray(Direct Array Index=&SpawnGroup(), Float Selection Array=DataInterface() {propertyValues {FloatData [-1,1]}})
         B=&Subtract_Float(
          A=&Multiply_Float(A=2, B=&Subtract_Float(A=$Engine.Time, B=0.5))
          B=$Particles.NormalizedAge
         )
        )
        Period=0.25
        Scale=&FloatFromCurve(FloatCurve=DataInterface() {propertyValues {Curve {keys [<at 0.5 4 RCIM_Cubic>,<at 1 0 RCIM_Cubic>]}}})
        Bias=0
       )
       Z=&Sine(
        Normalized Angle=&Multiply_Float(
         A=1
         B=&Subtract_Float(
          A=&Multiply_Float(A=2, B=&Subtract_Float(A=$Engine.Time, B=0.5))
          B=$Particles.NormalizedAge
         )
        )
        Period=0.25
        Scale=&FloatFromCurve(FloatCurve=DataInterface() {propertyValues {Curve {keys [<at 0.5 4 RCIM_Cubic>,<at 1 0 RCIM_Cubic>]}}})
        Bias=0
       )
      )
      Destination Space=::Simulation
      Apply Scale=true
     )
    )) {enabled=false, isSetParameters=true}
   /SetVariables(Particles.CustomPosition:NiagaraPosition=&AddVectorToPosition(
     Position=$Particles.Position
     Vector=&TransformVector(
      Vector=&MakeVector(
       X=0
       Y=&Sine(
        Normalized Angle=&Multiply_Float(
         A=&SelectFloatFromArray(Direct Array Index=&SpawnGroup(), Float Selection Array=DataInterface() {propertyValues {FloatData [-1,1]}})
         B=&Subtract_Float(
          A=&Multiply_Float(A=2, B=&Subtract_Float(A=$Engine.Time, B=0.5))
          B=$Particles.NormalizedAge
         )
        )
        Period=0.25
        Scale=&FloatFromCurve(FloatCurve=DataInterface() {propertyValues {Curve {keys [<at 0.5 4 RCIM_Cubic>,<at 1 0 RCIM_Cubic>]}}})
        Bias=0
       )
       Z=&Sine(
        Normalized Angle=&Multiply_Float(
         A=1
         B=&Subtract_Float(
          A=&Multiply_Float(A=2, B=&Subtract_Float(A=$Engine.Time, B=0.5))
          B=$Particles.NormalizedAge
         )
        )
        Period=0.25
        Scale=&FloatFromCurve(FloatCurve=DataInterface() {propertyValues {Curve {keys [<at 0.5 4 RCIM_Cubic>,<at 1 0 RCIM_Cubic>]}}})
        Bias=0
       )
      )
      Destination Space=::Simulation
      Apply Scale=true
     )
    )) {isSetParameters=true}
  ]
  rendererClasses [&NiagaraRibbonRendererProperties]
  renderers [
   {rendererClass=&NiagaraRibbonRendererProperties}
   {rendererClass=&NiagaraRibbonRendererProperties, rendererIndex=1}
  ]
 }
 NE_VC_Boost_Ribbon_Parent008 {
  emitterUpdateScript [
   &EmitterState()
   &VC_FilteredSocketCount()
   &VC_RateQualityHandler(High Spawn Rate=60, Low Spawn Rate=60, Medium Spawn Rate=60, Use Custom Rate?=true)
   &VC_SpawnRibbon_v1(Number of Ribbons=$Emitter.PortCount, SpawnRate=$Emitter.Spawn Rate)
   /SetVariables(Emitter.TimeOffset:NiagaraFloat=&Power(a=HlslExpression() {hlslExpression="(fmod(Engine.DeltaTime * 0.35, 0.2) + 0.0) * 3"}, b=2)) {isSetParameters=true}
   /SetVariables(
    Emitter.CustomDesaturationAmnt:NiagaraFloat=&MakeCustomFloatFromBool(Bool=$User.bIsPainted, True Float=1, False Float=0)
    Emitter.CustomDesaturationCtrst:NiagaraFloat=&MakeCustomFloatFromBool(Bool=$User.bIsPainted, True Float=1, False Float=0)
    Emitter.CustomDiffuseMode:NiagaraFloat=&MakeCustomFloatFromBool(Bool=$User.bIsPainted, True Float=0, False Float=1)
   ) {isSetParameters=true}
  ]
  particleSpawnScript [
   &VC_ExtraBindingColors(
    ExtraColors=1
    R1 Color:LinearColor=&MakeCustomLinearColorFromBool(Bool=$User.bIsPainted, True Color=$User.PaintedColor, False Color=<rgba 0 0 0 1>)
   )
   &VC_PaintedVariantsOverride(
    R1 Black:LinearColor=<rgba 0 0 0 1>
    R1 BurntSienna:LinearColor=<rgba 0.375 0.093 0.0225 1>
    R1 Cobalt:LinearColor=<rgba 0 0.0667 2 1>
    R1 Crimson:LinearColor=<rgba 0.564 0.0113 0.0263 1>
    R1 ForestGreen:LinearColor=<rgba 0.05 0.6 0 1>
    R1 Grey:LinearColor=<rgba 0.2864 0.2864 0.33 1>
    R1 Lime:LinearColor=<rgba 0.5367 0.7 0 1>
    R1 Orange:LinearColor=<rgba 1 0.1556 0 1>
    R1 Pink:LinearColor=<rgba 0.9 0.288 0.9 1>
    R1 Purple:LinearColor=<rgba 0.2945 0 0.6795 1>
    R1 Saffron:LinearColor=<rgba 0.94 0.7294 0.0376 1>
    R1 SkyBlue:LinearColor=<rgba 0 0.5722 0.86 1>
    R1 TitaniumWhite:LinearColor=<rgba 0.82 0.82 0.82 1>
    Use R1 Painted Overrides=true
   )
   &VC_InitializeRibbon_v1(
    Color=&MakeCustomLinearColorFromBool(Bool=$User.bIsPainted, True Color=$User.PaintedColor, False Color=<rgba 0 0 0 1>)
    Custom Offset=true
    Enable Skeletal Position=true
    Lifetime=0.7
    Local Vector Offset=<20 0 0>
    Offset Amount=1
    Ribbon Width:NiagaraFloat=&RandomRangeFloat(Minimum=64, Maximum=64)
   )
   &AddVelocity(Velocity=<-2000 0 0>) {module=AddVelocity001}
   &ShapeLocation(MaskSpawnGroup=false, Override Seed=false, Sphere Radius=2, UseOwnerScale=true) {enabled=false}
  ]
  particleUpdateScript [
   &ParticleState()
   &ScaleRibbonWidth(Ribbon Width Scale=&FloatFromCurve(FloatCurve=DataInterface() {propertyValues {Curve {keys [
         {arriveTangent=1.0826, interpMode=RCIM_Cubic, leaveTangent=1.0826, tangentMode=RCTM_User, time=0, value=0.5}
         {arriveTangent=0.0088, interpMode=RCIM_Cubic, leaveTangent=0.0088, tangentMode=RCTM_User, time=0.25, value=0.7}
         {arriveTangent=-1.7516, interpMode=RCIM_Cubic, leaveTangent=-1.7516, tangentMode=RCTM_User, time=0.8, value=0}
        ]}}}))
   &SolveForcesAndVelocity(Clamp Velocity=false, Limit Acceleration=false)
   &VC_UpdateRibbon_v1(Enable Skeletal Position=true)
   /SetVariables(Particles.Position=&AddVectorToPosition(
     Position=$Particles.Position
     Vector=&TransformVector(
      Vector=&MakeVector(
       X=0
       Y=&Sine(
        Normalized Angle=&Multiply_Float(
         A=&SelectFloatFromArray(Direct Array Index=&SpawnGroup(), Float Selection Array=DataInterface() {propertyValues {FloatData [-1,1]}})
         B=&Subtract_Float(
          A=&Multiply_Float(A=2, B=&Subtract_Float(A=$Engine.Time, B=0.5))
          B=$Particles.NormalizedAge
         )
        )
        Period=1
        Scale=&FloatFromCurve(FloatCurve=DataInterface() {propertyValues {Curve {keys [<at 0 2 RCIM_Cubic>,<at 1 0 RCIM_Cubic>]}}})
        Bias=0
       )
       Z=&Sine(
        Normalized Angle=&Add_Float(
         A=&Multiply_Float(
          A=$Particles.Age
          B=&SelectFloatFromArray(Direct Array Index=$Particles.SpawnGroup, Float Selection Array=DataInterface() {propertyValues {FloatData [-1,1]}})
         )
         B=&Multiply_Float(A=$Emitter.LoopedAge, B=0)
        )
        Period=$Emitter.TimeOffset
        Scale=&FloatFromCurve(
         FloatCurve=DataInterface() {propertyValues {Curve {keys [{arriveTangent=0.7811, interpMode=RCIM_Cubic, leaveTangent=0.7811, tangentMode=RCTM_User, time=0, value=0.85}, <at 0.5 1>]}}}
         Scale Curve=5
        )
        Bias=0
       )
      )
      Destination Space=::Simulation
      Apply Scale=true
     )
    )) {enabled=false, isSetParameters=true}
   /SetVariables(Particles.Position=&AddVectorToPosition(
     Position=$Particles.Position
     Vector=&TransformVector(
      Vector=&MakeVector(
       X=0
       Y=&Sine(
        Normalized Angle=&Multiply_Float(
         A=1
         B=&Subtract_Float(
          A=&Multiply_Float(A=2, B=&Subtract_Float(A=$Engine.Time, B=0.5))
          B=$Particles.NormalizedAge
         )
        )
        Period=&SelectFloatFromArray(
         Direct Array Index=&SpawnGroup()
         Float Selection Array=DataInterface() {propertyValues {FloatData [0.7,0.765]}}
        )
        Scale=&FloatFromCurve(FloatCurve=DataInterface() {propertyValues {Curve {keys [<at 0.1 2.5 RCIM_Cubic>,<at 0.86 0 RCIM_Cubic>]}}})
        Bias=0
       )
       Z=&Sine(
        Normalized Angle=&Multiply_Float(
         A=1
         B=&Subtract_Float(
          A=&Multiply_Float(A=2, B=&Subtract_Float(A=$Engine.Time, B=0.5))
          B=$Particles.NormalizedAge
         )
        )
        Period=&SelectFloatFromArray(
         Direct Array Index=&SpawnGroup()
         Float Selection Array=DataInterface() {propertyValues {FloatData [0.3,0.365]}}
        )
        Scale=&FloatFromCurve(FloatCurve=DataInterface() {propertyValues {Curve {keys [<at 0.1 3 RCIM_Cubic>,<at 0.86 0 RCIM_Cubic>]}}})
        Bias=0
       )
      )
      Destination Space=::Simulation
      Apply Scale=true
     )
    )) {isSetParameters=true}
  ]
  rendererClasses [&NiagaraRibbonRendererProperties]
  renderers [{rendererClass=&NiagaraRibbonRendererProperties}]
 }
 NE_VC_Boost_Sprite_Parent002 {
  bEnabled=false
  emitterUpdateScript [
   &EmitterState(Life Cycle Mode=::NewEnumerator1, UseLoopDelay=false)
   &VC_FilteredSocketCount()
   &SpawnRate(
    SpawnRate=&MakeCustomFloatFromBool(Bool=$User.bIsFrontendPreview, True Float=60, False Float=60)
    Use Spawn Probability=false
   )
  ]
  particleSpawnScript [
   &VC_PaintedVariantsOverride() {enabled=false}
   &InitializeParticle(
    Color=&MakeCustomLinearColorFromBool(Bool=$User.bIsPainted, True Color=$User.PaintedColor, False Color=<rgba 15 2.5 10 1>)
    Lifetime=&RandomRangeFloat(Minimum=0.25, Maximum=0.5)
    Sprite Rotation Mode=::NewEnumerator2
    Sprite Size Mode=::NewEnumerator3
    Uniform Sprite Size=16
    UsePositionOffset=false
    Write Material Random=true
   )
   &PartitionParticles(Number Of Partitions=$Emitter.PortCount)
   &SkeletalMeshLocation(
    Bone / Socket Index=$Particles.PartitionParticles.Partition
    Bone / Socket Sampling Mode=::NewEnumerator5
    MaskSpawnGroup=false
    Mesh Sampling Type=::NewEnumerator6
    Sampled Position Offset=<-50 0 0>
    Skeletal Mesh=$System.SkeletalMesh
   )
   &AddVelocity(
    Velocity=&MakeCustomVectorFromBool(Bool=$User.bIsFrontendPreview, True Vector=<1000 0 0>, False Vector=<1000 0 0>)
    Velocity Speed Scale=-1
   )
   &ShapeLocation(MaskSpawnGroup=false, Override Seed=false, Sphere Radius=25, UseOwnerScale=true)
  ]
  particleUpdateScript [
   &ParticleState()
   &ScaleColor(
    Scale Alpha=&MakeCustomFloatFromBool(
     Bool=$User.bIsPainted
     True Float=1
     False Float=&FloatFromCurve(FloatCurve=DataInterface() {propertyValues {Curve {keys [<at 0 0>,<at 0.05 1>,<at 0.75 1>,<at 1 0>]}}})
    )
    Scale RGB=&MakeCustomVectorFromBool(Bool=$User.bIsPainted, True Vector=<1 1 1>, False Vector=<1 1 1>)
    ScaleA=true
    ScaleRGB=true
   )
   &ScaleSpriteSize(Uniform Curve Sprite Scale=DataInterface() {propertyValues {Curve {keys [{tangentMode=RCTM_User, time=0, value=0.5}, <at 0.1 1.5>, <at 0.5 1.5>, {tangentMode=RCTM_User, time=1, value=0}]}}})
   &CurlNoiseForce(Noise Frequency=300, Noise Strength=750, Pan Noise=false)
   &Drag(Use Linear Drag=true, Use Rotational Drag=false)
   &SolveForcesAndVelocity(Clamp Velocity=false, Limit Acceleration=false)
  ]
  rendererClasses [&NiagaraSpriteRendererProperties]
  renderers [{rendererClass=&NiagaraSpriteRendererProperties}]
 }
 NE_VC_Boost_Sprite_Parent004 {
  emitterUpdateScript [
   &EmitterState()
   &VC_FilteredSocketCount()
   &SpawnRate(SpawnRate=10, Use Spawn Probability=false)
   &SpawnRate(SpawnGroup=1, SpawnRate=10, Use Spawn Probability=false) {module=SpawnRate001}
  ]
  particleSpawnScript [
   /SetVariables(Particles.VisibilityTag=&SpawnGroup()) {isSetParameters=true}
   &VC_ExtraBindingColors(
    ExtraColors=1
    R1 Color:LinearColor=&MakeCustomLinearColorFromBool(Bool=$User.bIsPainted, True Color=$User.PaintedColor, False Color=<rgba 2.0785 100 52.0952 1>)
   )
   &VC_PaintedVariantsOverride(
    R1 Black:LinearColor=<rgba 0.012 0.012 0.012 1>
    R1 BurntSienna:LinearColor=<rgba 0.8 0.1728 0.016 1>
    R1 Cobalt:LinearColor=<rgba 0 0.2667 8 1>
    R1 Crimson:LinearColor=<rgba 7 0 0.1897 1>
    R1 DefaultColor=<rgba 2.0785 100 52.0952 1>
    R1 ForestGreen:LinearColor=<rgba 0 2 0 1>
    R1 Grey:LinearColor=<rgba 0.2169 0.2169 0.25 1>
    R1 Lime:LinearColor=<rgba 4.9333 8 0 1>
    R1 Orange:LinearColor=<rgba 17 3.9667 0 1>
    R1 Pink:LinearColor=<rgba 15 3 15 1>
    R1 Purple:LinearColor=<rgba 0.52 0 1.2 1>
    R1 Saffron:LinearColor=<rgba 6 4 0 1>
    R1 SkyBlue:LinearColor=<rgba 0 1.6635 2.5 1>
    R1 TitaniumWhite:LinearColor=<rgba 1 1 1 1>
    Use R1 Painted Overrides=true
   )
   &InitializeParticle(
    Color=&MakeCustomLinearColorFromBool(Bool=$User.bIsPainted, True Color=$User.PaintedColor, False Color=<rgba 2.0785 100 52.0952 1>)
    Lifetime=&RandomRangeFloat(Minimum=0.5, Maximum=0.75)
    Sprite Rotation Mode=::NewEnumerator2
    Sprite Size Mode=::NewEnumerator3
    Uniform Sprite Size=&SelectFloatFromArray(Direct Array Index=&SpawnGroup(), Float Selection Array=DataInterface() {propertyValues {FloatData [12,8]}})
    UsePositionOffset=false
    Write Material Random=true
   )
   &PartitionParticles(Number Of Partitions=$Emitter.PortCount, Partition Mode:ENiagara_PartitionMode=::NewEnumerator5)
   &SkeletalMeshLocation(
    Bone / Socket Index=$Particles.PartitionParticles.Partition
    Bone / Socket Sampling Mode=::NewEnumerator5
    MaskSpawnGroup=false
    Mesh Sampling Type=::NewEnumerator6
    Skeletal Mesh=$System.SkeletalMesh
   )
   &AddVelocity(Velocity=<-750 0 0>)
   &ShapeLocation(MaskSpawnGroup=true, Override Seed=false, Sphere Radius=30, UseOwnerScale=true)
   &ShapeLocation(MaskSpawnGroup=true, Override Seed=false, Spawn Group Mask=1, Sphere Radius=25, UseOwnerScale=true) {module=ShapeLocation001}
  ]
  particleUpdateScript [
   &ParticleState()
   /SetVariables(Particles.R1 Color:LinearColor=&MultiplyLinearColorByVectorAndFloat(
     Color to Scale:LinearColor=$Particles.Initial.R1 Color
     RGB Scale:Vector3f=<1 1 1>
     Alpha Scale:NiagaraFloat=&FloatFromCurve(FloatCurve=DataInterface() {propertyValues {Curve {keys [<at 0 0>,<at 0.05 1>,<at 0.75 1>,<at 1 0>]}}})
    )) {isSetParameters=true}
   &ScaleColor(
    Scale Alpha=&FloatFromCurve(FloatCurve=DataInterface() {propertyValues {Curve {keys [<at 0 0>,<at 0.05 1>,<at 0.75 1>,<at 1 0>]}}})
    ScaleA=true
    ScaleRGB=true
   ) {enabled=false}
   &ScaleSpriteSize(Uniform Curve Sprite Scale=DataInterface() {propertyValues {Curve {keys [<at 0 0 RCIM_Cubic>, {arriveTangent=2, time=0.1, value=1.5}, <at 0.5 1.5>, {tangentMode=RCTM_User, time=1, value=0}]}}})
   &CurlNoiseForce(Bias Noise=false, Noise Frequency=20, Noise Strength=12000, Pan Noise=false, Pan Noise Field=<0.23 0.84 0.47>)
   &Drag(Drag=0.5, Use Linear Drag=true, Use Rotational Drag=false)
   &ScaleVelocity(
    Coordinate Space:ENiagaraCoordinateSpace=::Simulation
    Velocity Scale=&VectorFromCurve(VectorCurve=DataInterface() {propertyValues {
       XCurve {keys [<at 0.5 1 RCIM_Cubic>,<at 1 0 RCIM_Cubic>]}
       YCurve {keys [<at 0.5 1 RCIM_Cubic>,<at 1 0 RCIM_Cubic>]}
       ZCurve {keys [<at 0.5 1 RCIM_Cubic>,<at 1 0 RCIM_Cubic>]}
      }})
   )
   &SolveForcesAndVelocity(Clamp Velocity=false, Limit Acceleration=false)
  ]
  rendererClasses [&NiagaraSpriteRendererProperties]
  renderers [
   {rendererClass=&NiagaraSpriteRendererProperties}
   {rendererClass=&NiagaraSpriteRendererProperties, rendererIndex=1}
   {rendererClass=&NiagaraSpriteRendererProperties, rendererIndex=2}
  ]
 }
}
summary {
 systemName=NS_Boost_Electrobeam
 emitters [
  {bEnabled=false, emitterName=NE_VC_Boost_Sprite_Parent002, rendererClasses [&NiagaraSpriteRendererProperties]}
  {emitterName=NE_VC_Boost_Ribbon_Parent007, rendererClasses [&NiagaraRibbonRendererProperties]}
  {emitterName=NE_VC_Boost_Sprite_Parent004, rendererClasses [&NiagaraSpriteRendererProperties]}
  {emitterName=NE_VC_Boost_Ribbon_Parent008, rendererClasses [&NiagaraRibbonRendererProperties]}
  {emitterName=NE_VC_Boost_Ribbon_Parent, rendererClasses [&NiagaraRibbonRendererProperties]}
 ]
 userVariables [
  User.BelowTrailColor {type=LinearColor, description="", defaultValue {struct=LinearColor, value=<rgba 1 1 1 1>}}
  User.bIsFrontendPreview {type=NiagaraBool}
  User.bIsPainted {type=NiagaraBool}
  User.ColorVariantID {type=NiagaraInt32}
  User.PaintedColor {type=LinearColor, description="", defaultValue {struct=LinearColor, value=<rgba 1 1 1 1>}}
  User.PaintedScalar {type=NiagaraFloat, defaultValue=1, description=""}
  User.PortCount {type=NiagaraInt32, defaultValue=1}
  User.SizeScaling {type=NiagaraFloat, defaultValue=8, description=""}
  User.SubTrailColor {type=LinearColor, description="", defaultValue {struct=LinearColor, value=<rgba 1 1 1 1>}}
  User.TopTrailColor {type=LinearColor, description="", defaultValue {struct=LinearColor, value=<rgba 1 1 1 1>}}
  User.TrackTrailColor {type=LinearColor, description="", defaultValue {struct=LinearColor, value=<rgba 1 1 1 1>}}
 ]
}
