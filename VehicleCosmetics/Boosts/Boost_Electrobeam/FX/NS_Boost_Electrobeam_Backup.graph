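# Niagara system for the Electrobeam vehicle boost cosmetic, named as a backup, drawing trails and sparkle sprites from skeletal-mesh boost sockets.
# Four ribbon emitters layer trails of differing width and color per filtered socket, while two sprite emitters partition particles across sockets and scatter them with curl noise.
# User.bIsPainted switches colors and some fade curves between default and painted looks.
# The emitter "New" computes time-driven flicker values and uses a large lateral offset.
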NiagaraSystem /VehicleCosmetics/Boosts/Boost_Electrobeam/FX/NS_Boost_Electrobeam_Backup
dependencies {
 usedDataInterfaces [&NiagaraDataInterfaceVectorCurve, &NiagaraDataInterfaceCurve, &NiagaraDataInterfaceSkeletalMesh]
 usedDynamicInputs []
 usedModules [
  &self:SystemSpawnScript.SystemScriptSource.SystemScriptGraph.NiagaraNodeAssignment_0.SetVariables_#1
  &SystemState
 ]
}
emitters {
 NE_VC_Boost_Ribbon_Parent {
  emitterUpdateScript [
   &EmitterState(Life Cycle Mode=::NewEnumerator1, Loop Duration=1, UseLoopDelay=false)
   &VC_FilteredSocketCount()
   &VC_SpawnRibbon_v1(
    Number of Ribbons=$Emitter.PortCount
    SpawnRate=&MakeCustomFloatFromBool(Bool=$User.bIsFrontendPreview, True Float=120, False Float=120)
   )
  ]
  particleSpawnScript [
   &VC_PaintedVariantsOverride() {enabled=false}
   &VC_InitializeRibbon_v1(
    Color=&MakeCustomLinearColorFromBool(Bool=$User.bIsPainted, True Color=$User.TopTrailColor, False Color=<rgba 1 1 1 1>)
    Enable Skeletal Position=true
    Lifetime=0.65
    Ribbon Width:NiagaraFloat=&RandomRangeFloat(
     Minimum=&Multiply_Float(A=$User.SizeScaling, B=6)
     Maximum=&Multiply_Float(A=$User.SizeScaling, B=8)
    )
   )
   &AddVelocity(
    Velocity=&MakeCustomVectorFromBool(Bool=$User.bIsFrontendPreview, True Vector=<1000 0 0>, False Vector=<1000 0 0>)
    Velocity Speed Scale=-2
   ) {module=AddVelocity001}
   &ShapeLocation(MaskSpawnGroup=false, Override Seed=false, Sphere Radius=2, UseOwnerScale=true)
  ]
  particleUpdateScript [
   &ParticleState()
   &ScaleColor(
    Scale Alpha=&MakeCustomFloatFromBool(
     Bool=$User.bIsPainted
     True Float=&FloatFromCurve(FloatCurve=DataInterface() {propertyValues {Curve {keys [
          {arriveTangent=-0.0518, interpMode=RCIM_Cubic, leaveTangent=-0.0518, tangentMode=RCTM_User, time=0, value=0}
          {arriveTangent=4, interpMode=RCIM_Cubic, leaveTangent=4, time=0.1, value=0.55}
          {interpMode=RCIM_Cubic, tangentMode=RCTM_User, time=0.25, value=1}
          {arriveTangent=0.1106, interpMode=RCIM_Cubic, leaveTangent=0.1106, tangentMode=RCTM_User, time=1, value=1}
         ]}}})
     False Float=&FloatFromCurve(FloatCurve=DataInterface() {propertyValues {Curve {keys [
          {arriveTangent=-0.0518, interpMode=RCIM_Cubic, leaveTangent=-0.0518, tangentMode=RCTM_User, time=0, value=0}
          {arriveTangent=4, interpMode=RCIM_Cubic, leaveTangent=4, time=0.15, value=0.25}
          {interpMode=RCIM_Cubic, tangentMode=RCTM_User, time=0.25, value=1}
          {arriveTangent=0.1106, interpMode=RCIM_Cubic, leaveTangent=0.1106, tangentMode=RCTM_User, time=1, value=1}
         ]}}})
    )
    Scale RGB=&MakeCustomVectorFromBool(
     Bool=$User.bIsPainted
     True Vector=<1 1 1>
     False Vector=&VectorFromCurve(VectorCurve=DataInterface() {propertyValues {XCurve {keys [<at 0 2>,<at 1 1>]}, YCurve {keys [<at 0 2>,<at 1 1>]}, ZCurve {keys [<at 0 2>,<at 1 1>]}}})
    )
    ScaleA=true
    ScaleRGB=true
   )
   &ScaleRibbonWidth(Ribbon Width Scale=&FloatFromCurve(FloatCurve=DataInterface() {propertyValues {Curve {keys [
         <at 0 0>
         {arriveTangent=-2.211, interpMode=RCIM_Cubic, leaveTangent=-2.211, tangentMode=RCTM_User, time=0.25, value=0.85}
         <at 0.8 0 RCIM_Cubic>
        ]}}}))
   &DynamicMaterialParameters(
    Index 0 Param 1=&FloatFromCurve(FloatCurve=DataInterface() {propertyValues {Curve {keys [<at 0 2>,<at 0.7 0>]}}})
    Index 0 Param 2=&MakeCustomFloatFromBool(Bool=$User.bIsPainted, True Float=1, False Float=0)
    Module.Index0Param0Write=true
    Module.Index0Param1Write=true
    Module.Index0Param2Write=true
    Module.Index0Param3Write=true
   )
   &SolveForcesAndVelocity(Clamp Velocity=false, Limit Acceleration=false)
   &VC_UpdateRibbon_v1(Enable Skeletal Position=true)
  ]
  rendererClasses [&NiagaraRibbonRendererProperties]
  renderers [{rendererClass=&NiagaraRibbonRendererProperties}]
 }
 NE_VC_Boost_Ribbon_Parent001 {
  emitterSpawnScript [/SetVariables(
    Emitter.IsPainted:NiagaraFloat=&MakeCustomFloatFromBool(Bool=$User.bIsPainted, True Float=2, False Float=-0.512)
    Emitter.MirrorOffset:NiagaraFloat=&MakeCustomFloatFromBool(Bool=$User.bIsPainted, True Float=4, False Float=2)
   ) {isSetParameters=true, module=SetVariables_#2}]
  emitterUpdateScript [
   &EmitterState(Life Cycle Mode=::NewEnumerator1, Loop Duration=1, UseLoopDelay=false)
   &VC_FilteredSocketCount()
   &VC_SpawnRibbon_v1(
    Number of Ribbons=$Emitter.PortCount
    SpawnRate=&MakeCustomFloatFromBool(Bool=$User.bIsFrontendPreview, True Float=120, False Float=120)
   )
  ]
  particleSpawnScript [
   &VC_PaintedVariantsOverride() {enabled=false}
   &VC_InitializeRibbon_v1(
    Color=&MakeCustomLinearColorFromBool(Bool=$User.bIsPainted, True Color=$User.BelowTrailColor, False Color=<rgba 50 6.4426 41.9262 1>)
    Custom Offset=true
    Enable Skeletal Position=true
    Lifetime=0.46
    Local Vector Offset=<0 0 1>
    Offset Amount:NiagaraFloat=&MakeFloatFromVector(Vector=<0 0 0>, Channel=::NewEnumerator2)
    Ribbon Width=60
   )
   &AddVelocity(
    Velocity=&MakeCustomVectorFromBool(Bool=$User.bIsFrontendPreview, True Vector=<1000 0 0>, False Vector=<1000 0 0>)
    Velocity Speed Scale=-2
   ) {module=AddVelocity001}
   &ShapeLocation(MaskSpawnGroup=false, Override Seed=false, Sphere Radius=7, UseOwnerScale=true)
  ]
  particleUpdateScript [
   &ParticleState()
   &ScaleColor(
    Scale Alpha=&MakeCustomFloatFromBool(
     Bool=$User.bIsPainted
     True Float=0.15
     False Float=&FloatFromCurve(FloatCurve=DataInterface() {propertyValues {Curve {keys [<at 0.05 0>,<at 0.15 0.1>,<at 0.9 0.1>,<at 1 0>]}}}, Scale Curve=0.2)
    )
    Scale RGB=&MakeCustomVectorFromBool(Bool=$User.bIsPainted, True Vector=<2 2 2>, False Vector=<2 2 2>)
    ScaleA=true
    ScaleRGB=true
   )
   &ScaleRibbonWidth(Ribbon Width Scale=&FloatFromCurve(
     FloatCurve=DataInterface() {propertyValues {Curve {keys [
         {interpMode=RCIM_Cubic, tangentMode=RCTM_User, time=0, value=0}
         {arriveTangent=0.5028, interpMode=RCIM_Cubic, leaveTangent=0.5028, tangentMode=RCTM_User, time=0.1, value=1.25}
         {arriveTangent=-0.5446, interpMode=RCIM_Cubic, leaveTangent=-0.5446, tangentMode=RCTM_User, time=0.8, value=1}
         {arriveTangent=-3.985, interpMode=RCIM_Cubic, leaveTangent=-3.985, tangentMode=RCTM_User, time=0.9, value=0}
        ]}}}
     Scale Curve=1.25
    ))
   &DynamicMaterialParameters(
    Index 0 Param 1=1
    Index 0 Param 2=0.05
    Module.Index0Param0Write=true
    Module.Index0Param1Write=true
    Module.Index0Param2Write=true
    Module.Index0Param3Write=true
   ) {enabled=false}
   &SolveForcesAndVelocity(Clamp Velocity=false, Limit Acceleration=false)
   &VC_UpdateRibbon_v1(Enable Skeletal Position=true)
  ]
  rendererClasses [&NiagaraRibbonRendererProperties]
  renderers [{rendererClass=&NiagaraRibbonRendererProperties}]
 }
 NE_VC_Boost_Ribbon_Parent002 {
  emitterUpdateScript [
   &EmitterState(Life Cycle Mode=::NewEnumerator1, Loop Duration=1, UseLoopDelay=false)
   &VC_FilteredSocketCount()
   &VC_SpawnRibbon_v1(
    Number of Ribbons=$Emitter.PortCount
    SpawnRate=&MakeCustomFloatFromBool(Bool=$User.bIsFrontendPreview, True Float=120, False Float=120)
   )
  ]
  particleSpawnScript [
   &VC_PaintedVariantsOverride() {enabled=false}
   &VC_InitializeRibbon_v1(
    Color=&MakeCustomLinearColorFromBool(Bool=$User.bIsPainted, True Color=$User.TrackTrailColor, False Color=<rgba 15 0 13.0312 1>)
    Custom Offset=true
    Enable Skeletal Position=true
    Lifetime=0.46
    Local Vector Offset=<1 0 0>
    Offset Amount:NiagaraFloat=&MakeFloatFromVector(Vector=<-1 0 0>, Channel=::NewEnumerator0)
    Ribbon Width:NiagaraFloat=&RandomRangeFloat(Minimum=20, Maximum=30)
   )
   &AddVelocity(
    Velocity=&MakeCustomVectorFromBool(Bool=$User.bIsFrontendPreview, True Vector=<1000 0 0>, False Vector=<1000 0 0>)
    Velocity Speed Scale=-2
   ) {module=AddVelocity001}
   &ShapeLocation(MaskSpawnGroup=false, Override Seed=false, Sphere Radius=10, UseOwnerScale=true)
  ]
  particleUpdateScript [
   &ParticleState()
   &ScaleColor(
    Scale Alpha=&MakeCustomFloatFromBool(
     Bool=$User.bIsPainted
     True Float=1
     False Float=&FloatFromCurve(FloatCurve=DataInterface() {propertyValues {Curve {keys [<at 0.05 0>,<at 0.15 1>,<at 0.9 1>,<at 1 0>]}}}, Scale Curve=0.35)
    )
    Scale RGB=&MakeCustomVectorFromBool(
     Bool=$User.bIsPainted
     True Vector=<1 1 1>
     False Vector=&VectorFromCurve(VectorCurve=DataInterface() {propertyValues {XCurve {keys [<at 0 7>,<at 1 0>]}, YCurve {keys [<at 0 7>,<at 1 0>]}, ZCurve {keys [<at 0 7>,<at 1 0>]}}})
    )
    ScaleA=true
    ScaleRGB=true
   )
   &ScaleRibbonWidth(Ribbon Width Scale=&FloatFromCurve(
     FloatCurve=DataInterface() {propertyValues {Curve {keys [
         {arriveTangent=-0.0009, interpMode=RCIM_Cubic, leaveTangent=-0.0009, tangentMode=RCTM_User, time=0, value=0}
         {arriveTangent=0.6917, interpMode=RCIM_Cubic, leaveTangent=0.6917, tangentMode=RCTM_User, time=0.5, value=1}
         {arriveTangent=-2.5286, interpMode=RCIM_Cubic, leaveTangent=-2.5286, tangentMode=RCTM_User, time=1, value=0}
        ]}}}
     Scale Curve=1.25
    ))
   &DynamicMaterialParameters(
    Index 0 Param 1=1
    Index 0 Param 2=0.1
    Module.Index0Param0Write=true
    Module.Index0Param1Write=true
    Module.Index0Param2Write=true
    Module.Index0Param3Write=true
   )
   &SolveForcesAndVelocity(Clamp Velocity=false, Limit Acceleration=false)
   &VC_UpdateRibbon_v1(Enable Skeletal Position=true)
  ]
  rendererClasses [&NiagaraRibbonRendererProperties]
  renderers [{rendererClass=&NiagaraRibbonRendererProperties}]
 }
 NE_VC_Boost_Sprite_Parent {
  emitterUpdateScript [
   &EmitterState(Life Cycle Mode=::NewEnumerator1, UseLoopDelay=false)
   &VC_FilteredSocketCount()
   &SpawnRate(
    SpawnRate=&MakeCustomFloatFromBool(Bool=$User.bIsFrontendPreview, True Float=30, False Float=30)
    Use Spawn Probability=false
   )
  ]
  particleSpawnScript [
   &VC_PaintedVariantsOverride() {enabled=false}
   &InitializeParticle(
    Color=&MakeCustomLinearColorFromBool(Bool=$User.bIsPainted, True Color=$User.PaintedColor, False Color=<rgba 15 2.5 10 1>)
    Lifetime=&RandomRangeFloat(Minimum=0.25, Maximum=0.5)
    Sprite Rotation Mode=::NewEnumerator2
    Sprite Size Mode=::NewEnumerator3
    Uniform Sprite Size=16
    UsePositionOffset=false
    Write Material Random=true
   )
   &PartitionParticles(Number Of Partitions=$Emitter.PortCount)
   &SkeletalMeshLocation(
    Bone / Socket Index=$Particles.PartitionParticles.Partition
    Bone / Socket Sampling Mode=::NewEnumerator5
    MaskSpawnGroup=false
    Mesh Sampling Type=::NewEnumerator6
    Sampled Position Offset=<-50 0 0>
    Skeletal Mesh=$System.SkeletalMesh
   )
   &AddVelocity(
    Velocity=&MakeCustomVectorFromBool(Bool=$User.bIsFrontendPreview, True Vector=<1000 0 0>, False Vector=<1000 0 0>)
    Velocity Speed Scale=-1
   )
   &ShapeLocation(MaskSpawnGroup=false, Override Seed=false, Sphere Radius=25, UseOwnerScale=true)
  ]
  particleUpdateScript [
   &ParticleState()
   &ScaleColor(
    Scale Alpha=&MakeCustomFloatFromBool(
     Bool=$User.bIsPainted
     True Float=1
     False Float=&FloatFromCurve(FloatCurve=DataInterface() {propertyValues {Curve {keys [<at 0 0>,<at 0.05 1>,<at 0.75 1>,<at 1 0>]}}})
    )
    Scale RGB=&MakeCustomVectorFromBool(Bool=$User.bIsPainted, True Vector=<1 1 1>, False Vector=<1 1 1>)
    ScaleA=true
    ScaleRGB=true
   )
   &ScaleSpriteSize(Uniform Curve Sprite Scale=DataInterface() {propertyValues {Curve {keys [{tangentMode=RCTM_User, time=0, value=0.5}, <at 0.1 1.5>, <at 0.5 1.5>, {tangentMode=RCTM_User, time=1, value=0}]}}})
   &CurlNoiseForce(Noise Frequency=300, Noise Strength=750, Pan Noise=false)
   &Drag(Use Linear Drag=true, Use Rotational Drag=false)
   &SolveForcesAndVelocity(Clamp Velocity=false, Limit Acceleration=false)
  ]
  rendererClasses [&NiagaraSpriteRendererProperties]
  renderers [{rendererClass=&NiagaraSpriteRendererProperties}]
 }
 NE_VC_Boost_Sprite_Parent001 {
  emitterUpdateScript [
   &EmitterState(Life Cycle Mode=::NewEnumerator1, UseLoopDelay=false)
   &VC_FilteredSocketCount()
   &SpawnRate(
    SpawnRate=&MakeCustomFloatFromBool(Bool=$User.bIsFrontendPreview, True Float=30, False Float=30)
    Use Spawn Probability=false
   )
  ]
  particleSpawnScript [
   &VC_PaintedVariantsOverride() {enabled=false}
   &InitializeParticle(
    Color=&MakeCustomLinearColorFromBool(Bool=$User.bIsPainted, True Color=$User.PaintedColor, False Color=<rgba 15 2.5 10 1>)
    Lifetime=&RandomRangeFloat(Minimum=0.25, Maximum=0.5)
    Sprite Rotation Mode=::NewEnumerator2
    Sprite Size Mode=::NewEnumerator3
    Uniform Sprite Size=12
    UsePositionOffset=false
    Write Material Random=true
   )
   &PartitionParticles(Number Of Partitions=$Emitter.PortCount)
   &SkeletalMeshLocation(
    Bone / Socket Index=$Particles.PartitionParticles.Partition
    Bone / Socket Sampling Mode=::NewEnumerator5
    MaskSpawnGroup=false
    Mesh Sampling Type=::NewEnumerator6
    Sampled Position Offset=<-50 0 0>
    Skeletal Mesh=$System.SkeletalMesh
   )
   &AddVelocity(
    Velocity=&MakeCustomVectorFromBool(Bool=$User.bIsFrontendPreview, True Vector=<1000 0 0>, False Vector=<1000 0 0>)
    Velocity Speed Scale=-1
   )
   &ShapeLocation(MaskSpawnGroup=false, Override Seed=false, Sphere Radius=20, UseOwnerScale=true)
  ]
  particleUpdateScript [
   &ParticleState()
   &ScaleColor(
    Scale Alpha=&FloatFromCurve(FloatCurve=DataInterface() {propertyValues {Curve {keys [<at 0 0>,<at 0.05 1>,<at 0.75 1>,<at 1 0>]}}})
    ScaleA=true
    ScaleRGB=true
   )
   &ScaleSpriteSize(Uniform Curve Sprite Scale=DataInterface() {propertyValues {Curve {keys [{tangentMode=RCTM_User, time=0, value=0.5}, <at 0.1 1.5>, <at 0.5 1.5>, {tangentMode=RCTM_User, time=1, value=0}]}}})
   &CurlNoiseForce(Noise Frequency=500, Noise Strength=1000, Pan Noise=false)
   &Drag(Use Linear Drag=true, Use Rotational Drag=false)
   &SolveForcesAndVelocity(Clamp Velocity=false, Limit Acceleration=false)
  ]
  rendererClasses [&NiagaraSpriteRendererProperties]
  renderers [{rendererClass=&NiagaraSpriteRendererProperties}]
 }
 New {
  emitterSpawnScript [/SetVariables(
    Emitter.IsPainted:NiagaraFloat=&MakeCustomFloatFromBool(Bool=$User.bIsPainted, True Float=2, False Float=-0.512)
    Emitter.MirrorOffset:NiagaraFloat=&MakeCustomFloatFromBool(Bool=$User.bIsPainted, True Float=4, False Float=2)
   ) {isSetParameters=true, module=SetVariables_#2}]
  emitterUpdateScript [
   &EmitterState(Life Cycle Mode=::NewEnumerator1, Loop Duration=1, UseLoopDelay=false)
   &VC_FilteredSocketCount()
   &VC_SpawnRibbon_v1(
    Number of Ribbons=$Emitter.PortCount
    SpawnRate=&MakeCustomFloatFromBool(Bool=$User.bIsFrontendPreview, True Float=120, False Float=120)
   )
   /SetVariables(
    Emitter.TimeLerp:NiagaraFloat=&Power(a=HlslExpression() {hlslExpression="(fmod( Engine.Time * 0.325 , 0.4) + 0.15) * 3"}, b=2)
    Emitter.TimeOffset:NiagaraFloat=&Power(a=HlslExpression() {hlslExpression="fmod( Engine.Time , 0.3) + 0.15 * 3"}, b=2)
   ) {isSetParameters=true}
   /SetVariables(
    Emitter.LerpChannel:Vector3f=&Lerp_Vector(
     A=<1 0 0>
     B=<0 1 0>
     Alpha=&Ceil_Float(Float=&Sine(Normalized Angle=$Emitter.Age, Period=$Emitter.TimeLerp, Scale=1, Bias=0))
    )
    Emitter.OffsetX:NiagaraFloat=&Multiply_Float(
     A=&Ceil_Float(Float=&Sine(Normalized Angle=$Emitter.Age, Period=$Emitter.TimeOffset, Scale=1, Bias=0))
     B=0.25
    )
   ) {isSetParameters=true}
  ]
  particleSpawnScript [
   &VC_PaintedVariantsOverride() {enabled=false}
   &VC_InitializeRibbon_v1(
    Color=&MakeCustomLinearColorFromBool(Bool=$User.bIsPainted, True Color=$User.BelowTrailColor, False Color=<rgba 15 3.7826 12.9207 1>)
    Custom Offset=true
    Enable Skeletal Position=true
    Lifetime=0.46
    Local Vector Offset=<0 150 0>
    Offset Amount=1
    Ribbon Width=60
   )
   &AddVelocity(
    Velocity=&MakeCustomVectorFromBool(Bool=$User.bIsFrontendPreview, True Vector=<1000 0 0>, False Vector=<1000 0 0>)
    Velocity Speed Scale=-2
   ) {module=AddVelocity001}
   &ShapeLocation(MaskSpawnGroup=false, Override Seed=false, Sphere Radius=7, UseOwnerScale=true)
  ]
  particleUpdateScript [
   &ParticleState()
   &ScaleColor(
    Scale Alpha=&MakeCustomFloatFromBool(
     Bool=$User.bIsPainted
     True Float=0.15
     False Float=&FloatFromCurve(FloatCurve=DataInterface() {propertyValues {Curve {keys [<at 0.05 0>,<at 0.15 0.1>,<at 0.9 0.1>,<at 1 0>]}}}, Scale Curve=0.2)
    )
    Scale RGB=&MakeCustomVectorFromBool(Bool=$User.bIsPainted, True Vector=<2 2 2>, False Vector=<2 2 2>)
    ScaleA=true
    ScaleRGB=true
   )
   &ScaleRibbonWidth(Ribbon Width Scale=&FloatFromCurve(
     FloatCurve=DataInterface() {propertyValues {Curve {keys [
         {interpMode=RCIM_Cubic, tangentMode=RCTM_User, time=0, value=0}
         {arriveTangent=0.5028, interpMode=RCIM_Cubic, leaveTangent=0.5028, tangentMode=RCTM_User, time=0.1, value=1.25}
         {arriveTangent=-0.5446, interpMode=RCIM_Cubic, leaveTangent=-0.5446, tangentMode=RCTM_User, time=0.8, value=1}
         {arriveTangent=-3.985, interpMode=RCIM_Cubic, leaveTangent=-3.985, tangentMode=RCTM_User, time=0.9, value=0}
        ]}}}
     Scale Curve=1.25
    ))
   &DynamicMaterialParameters(
    Index 0 Param 1=1
    Index 0 Param 2=0.05
    Module.Index0Param0Write=true
    Module.Index0Param1Write=true
    Module.Index0Param2Write=true
    Module.Index0Param3Write=true
   ) {enabled=false}
   &SolveForcesAndVelocity(Clamp Velocity=false, Limit Acceleration=false)
   &VC_UpdateRibbon_v1(Enable Skeletal Position=true)
  ]
  rendererClasses [&NiagaraRibbonRendererProperties]
  renderers [{rendererClass=&NiagaraRibbonRendererProperties}]
 }
}
summary {
 systemName=NS_Boost_Electrobeam_Backup
 emitters [
  {emitterName=NE_VC_Boost_Ribbon_Parent, rendererClasses [&NiagaraRibbonRendererProperties]}
  {emitterName=NE_VC_Boost_Ribbon_Parent001, rendererClasses [&NiagaraRibbonRendererProperties]}
  {emitterName=NE_VC_Boost_Ribbon_Parent002, rendererClasses [&NiagaraRibbonRendererProperties]}
  {emitterName=NE_VC_Boost_Sprite_Parent, rendererClasses [&NiagaraSpriteRendererProperties]}
  {emitterName=NE_VC_Boost_Sprite_Parent001, rendererClasses [&NiagaraSpriteRendererProperties]}
  {emitterName=New, rendererClasses [&NiagaraRibbonRendererProperties]}
 ]
 userVariables [
  User.BelowTrailColor {type=LinearColor, description="", defaultValue {struct=LinearColor, value=<rgba 1 1 1 1>}}
  User.bIsFrontendPreview {type=NiagaraBool}
  User.bIsPainted {type=NiagaraBool}
  User.ColorVariantID {type=NiagaraInt32}
  User.PaintedColor {type=LinearColor, description="", defaultValue {struct=LinearColor, value=<rgba 1 1 1 1>}}
  User.PaintedScalar {type=NiagaraFloat, defaultValue=1, description=""}
  User.PortCount {type=NiagaraInt32, defaultValue=1}
  User.SizeScaling {type=NiagaraFloat, defaultValue=8, description=""}
  User.SubTrailColor {type=LinearColor, description="", defaultValue {struct=LinearColor, value=<rgba 1 1 1 1>}}
  User.TopTrailColor {type=LinearColor, description="", defaultValue {struct=LinearColor, value=<rgba 1 1 1 1>}}
  User.TrackTrailColor {type=LinearColor, description="", defaultValue {struct=LinearColor, value=<rgba 1 1 1 1>}}
 ]
}
