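# Translucent unlit material for Niagara ribbons and mesh particles in the Electrobeam vehicle boost effect.
# Bright and dark colors blend across UV V, scaled by a dynamic Lerp parameter, then tint by particle color into an emissive-adjustment function.
# Particle alpha combines with a camera-proximity fade and depth fade.
# A noise function borrowed from the ElectroBeam drift trail gets a stepped, time-based flicker offset and is masked along UV1.
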Material /VehicleCosmetics/Boosts/Boost_Electrobeam/Materials/Ribbon_2_MAT
graph {
 properties {
  bUsedWithNiagaraMeshParticles=true
  bUsedWithNiagaraRibbons=true
  blendMode=BLEND_Translucent
  pixelDepthOffsetMode=PDOM_Legacy
  shadingModel=MSM_Unlit
  usedShadingModels=Unlit
 }
 nodes [
  MaterialExpressionVertexColor_1 = ParticleColor()
  MaterialExpressionCameraWorldPosition_0 = CameraPositionWS()
  n1 = WorldPosition()
  n2 = Distance(A=@MaterialExpressionCameraWorldPosition_0, B=@n1.XYZ)
  n3 = DotProduct(
   A=Normalize(VectorInput=Subtract(A=@MaterialExpressionCameraWorldPosition_0, B=@n1.XYZ))
   B=Transform(None=Constant3Vector() {constant=<rgba 0 0 -1 0>}) {transformSourceType=TRANSFORMSOURCE_View}
  )
  n4 = Divide(A=Subtract(A=Multiply(A=@n2, B=@n3), B=Constant() {r=1}), B=Constant() {r=50})
  MaterialExpressionDepthBiasBlend_0 = DepthFade()
  MaterialExpressionConstantClamp_3 = Saturate(None=Multiply(A=Multiply(A=@MaterialExpressionVertexColor_1.A, B=@n4), B=@MaterialExpressionDepthBiasBlend_0))
  Constant() {r=1}
  Constant() {r=8}
  n5 = VectorParameter() {parameterName=Color_Bright, defaultValue=<rgba 2 0.2 0 1>}
  n6 = VectorParameter() {parameterName=Color_Dark, defaultValue=<rgba 0.172 0 0.364 1>}
  n7 = ComponentMask(None=TextureCoordinate()) {g=true}
  n8 = OneMinus(None=@n7)
  n9 = Multiply(A=DynamicParameter() {paramNames [Lerp, "", Param3, Param4]}.Lerp, B=Multiply(A=Multiply(A=@n8, B=@n7), B=Constant() {r=4}))
  n10 = LinearInterpolate(A=@n5.R, B=@n6.R, Alpha=@n9)
  n11 = LinearInterpolate(A=@n5.RGB, B=@n6.RGB, Alpha=@n9)
  MaterialExpressionConstantClamp_2 = Saturate(None=Multiply(A=StaticSwitchParameter(True=@n10, False=@n11) {parameterName=PreColorDesaturate}, B=@MaterialExpressionVertexColor_1.RGB))
  MaterialFunctionCall(Emissive=@MaterialExpressionConstantClamp_2) {materialFunction=&MF_FX_EmissiveAdjustments}
  Multiply(A=Constant() {r=-1}, B=@n4)
  n12 = Multiply(A=Time(), B=ScalarParameter() {defaultValue=1, parameterName=OffsetSpeed})
  n13 = ScalarParameter() {defaultValue=0.3, parameterName=OffsetPeriod}
  n14 = Multiply(A=Add(A=Fmod(A=@n12, B=@n13), B=Constant() {r=0.15}), B=Constant() {r=3})
  n15 = Ceil(None=Sine(None=Multiply(A=@n14, B=@n14)))
  n16 = ScalarParameter() {defaultValue=0.25, parameterName=U_OffsetMult}
  n17 = ComponentMask(None=TextureCoordinate() {coordinateIndex=1}) {r=true}
  n18 = ScalarParameter() {defaultValue=2, parameterName=NoiseMaskMult}
  MaterialExpressionConstantClamp_4 = Saturate(None=Power(Base=Multiply(A=@n17, B=@n18), Exp=ScalarParameter() {defaultValue=0.35, parameterName=NoiseMaskPow}))
  Multiply(
   A=MaterialFunctionCall!(
    Speed=ScalarParameter() {defaultValue=6, parameterName=Speed}
    FlickerOffset=Multiply(A=@n15, B=@n16)
    ScaleX=ScalarParameter() {defaultValue=1, parameterName=ScaleX}
    ScaleY=ScalarParameter() {defaultValue=1, parameterName=ScaleY}
    ScaleZ=ScalarParameter() {defaultValue=1, parameterName=ScaleZ}
    Tile=ScalarParameter() {defaultValue=10, parameterName=tile}
   ) {materialFunction=/VehicleCosmetics/DriftTrail/DriftTrail_ElectroBeam/Materials/ElectroTrailNoise_MF}.Result
   B=@MaterialExpressionConstantClamp_4
  )
  Reroute(None=@MaterialExpressionVertexColor_1.A)
  StaticBool() {value=true}
  VertexColor()
 ]
}
